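; Editor utility data asset for the MetaSound Preset Migrator tool that stores saved input-name mappings between origin parent and destination MetaSound Sources.
; A Destination MetaSounds map, keyed by name, holds link structs, each nesting a map of origin MetaSound names to input-link maps.
; Functions check whether a pair exists, fetch its saved input links with a found flag, and save links by rewriting both map levels.
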(fn ContainsPairedMetaSounds
  (params (Origin%20MetaSound%20Source :type Meta%20Sound%20Source%20Object%20Reference) (Destination%20MetaSound%20Source :type Meta%20Sound%20Source%20Object%20Reference))
  (bind (_value _returnvalue) (Utilities|Map|Find (Variables|Default|GetDestination%20MetaSounds) (Utilities|String|StringToName (Utilities|GetDisplayName Destination%20MetaSound%20Source))))
  (bind _value_linkedoriginmetasounds_21_687fdea3440afdd825eb9e8f3651d5ed (Utilities|Struct|BreakMetaSoundPresetMigratorMetaSoundLinks _value))
  (bind (_value_2 _returnvalue_2) (Utilities|Map|Find _value_linkedoriginmetasounds_21_687fdea3440afdd825eb9e8f3651d5ed (Utilities|String|StringToName (Utilities|GetDisplayName Origin%20MetaSound%20Source))))
  (return (and _returnvalue_2 _returnvalue)))

(fn GetSavedInputLinks
  (params (Origin%20Parent%20MetaSound%20Source :type Meta%20Sound%20Source%20Object%20Reference) (Destination%20MetaSound%20Source :type Meta%20Sound%20Source%20Object%20Reference))
  (bind (_value _returnvalue) (Utilities|Map|Find (Variables|Default|GetDestination%20MetaSounds) (Utilities|GetMetaSoundName (Variables|GetDestination%20MetaSound%20Source))))
  (bind _value_linkedoriginmetasounds_21_687fdea3440afdd825eb9e8f3651d5ed (Utilities|Struct|BreakMetaSoundPresetMigratorMetaSoundLinks _value))
  (bind (_value_2 _returnvalue_2) (Utilities|Map|Find _value_linkedoriginmetasounds_21_687fdea3440afdd825eb9e8f3651d5ed (Utilities|GetMetaSoundName (Variables|GetOrigin%20Parent%20MetaSound%20Source))))
  (bind _value_linkedinputs_4_e2af70ab498d598692343a86473b2856 (Utilities|Struct|BreakMetaSoundPresetMigratorInputLinks _value_2))
  (return (and _returnvalue _returnvalue_2) _value_linkedinputs_4_e2af70ab498d598692343a86473b2856))

(fn SaveOriginMetaSoundInputLinks
  (params (Destination%20MetaSound :type Meta%20Sound%20Source%20Object%20Reference) (Origin%20Parent%20MetaSound :type Meta%20Sound%20Source%20Object%20Reference) (Linked%20Inputs :type Map%20of%20Names%20to%20Names))
  (bind _value (Utilities|Map|Find (Variables|Default|GetDestination%20MetaSounds) (Utilities|GetMetaSoundName (Variables|GetDestination%20MetaSound))))
  (bind _value_linkedoriginmetasounds_21_687fdea3440afdd825eb9e8f3651d5ed (Utilities|Struct|BreakMetaSoundPresetMigratorMetaSoundLinks _value))
  (bind _output_get (LocalVariables|SetLinked%20Origin%20MetaSounds _value_linkedoriginmetasounds_21_687fdea3440afdd825eb9e8f3651d5ed))
  (Utilities|Map|Add _output_get (Utilities|GetMetaSoundName (Variables|GetOrigin%20Parent%20MetaSound)) (Utilities|Struct|MakeMetaSoundPresetMigratorInputLinks (Variables|GetLinked%20Inputs)))
  (Utilities|Map|Add (Variables|Default|GetDestination%20MetaSounds) (Utilities|GetMetaSoundName (Variables|GetDestination%20MetaSound)) (Utilities|Struct|MakeMetaSoundPresetMigratorMetaSoundLinks (LocalVariables|GetLinked%20Origin%20MetaSounds))))
